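; Gameplay cue that plays the rez-out death effect for VividRazor minion reinforcement pawns on burst.
; It stores hit parameters and the pawn, plays an elimination montage, then after a delay stops looping audio and spawns dissolve VFX.
; The scaled-up Niagara effect spawns offset from a mesh socket, alongside a death sound at the pawn's location.
; A further short delay precedes hiding the pawn, and DissolveMaterialSetup is empty.
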(event EventOnBurst(EitherVFX-preferthisversion)
  (params (MyTarget :type Actor%20Object%20Reference) (Parameters :type Gameplay%20Cue%20Parameters%20Structure%20%28by%20ref%29) (ParticleComponents :type Array%20of%20FXSystem%20Component%20Object%20References) (AudioComponents :type Array%20of%20Audio%20Component%20Object%20References) (BurstCameraShakeInstance :type Camera%20Shake%20Base%20Object%20Reference) (BurstDecalInstance :type Decal%20Actor%20Object%20Reference))
  (CallFunction|SetonDeathHitParameters Parameters)
  (bind _asplayer_pawn_athena (Utilities|Casting|CastToPlayerPawn_Athena MyTarget))
  (Variables|DONOTEDIT|SetPawn _asplayer_pawn_athena)
  (CallFunction|PlayElminationAnimMontage)
  (Utilities|FlowControl|Delay (Variables|DONOTEDIT|GetPawn) (Variables|DISSOLVE|GetDelayBetweenAnimAndVFX))
  (CallFunction|StopLoopingAudio)
  (CallFunction|SpawnDissolveVFX)
  (Utilities|FlowControl|Delay (Variables|DONOTEDIT|GetPawn) 0.5) ; Add a Potential Delay for IPs (Such as Realm/Vader) where we want to see him animate more when he dies.
  (Rendering|SetActorHiddenInGame (Variables|DONOTEDIT|GetPawn) true))

(fn DissolveMaterialSetup)

(fn SpawnDissolveVFX
  (bind _returnvalue (Niagara|SpawnSystematLocation 0 (Variables|DISSOLVE|GetDissolve%20Visual%20Effect) (- (+ (Transformation|GetSocketLocation (Variables|/Script/Engine.Character|GetMesh (Variables|DONOTEDIT|GetPawn)) (Variables|DISSOLVE|GetDissolve%20VFX%20Spawn%20Point%20Name)) (* (Transformation|GetActorForwardVector (Variables|DONOTEDIT|GetPawn)) 25.0)) "0.000000,0.000000,115.000000")))
  (Niagara|SetNiagaraVariable(Float) _returnvalue "Global Scale" 2.5)
  (Audio|SpawnSoundatLocation 0 (Variables|Default|GetSoundonNPCDeath) (Transformation|GetActorLocation (Variables|DONOTEDIT|GetPawn))))

(fn ConstructionScript
  (|CallParentFunction))
